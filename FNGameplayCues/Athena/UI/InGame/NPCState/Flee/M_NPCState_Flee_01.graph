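# Unlit translucent material for the animated flee-state icon of NPCs, with camera alignment and distance-based scaling math.
# Curve atlases animate the four channels of one mask texture as looping sweat scaling with an end-of-loop fade, or as jitter by rotation or scale via static switch.
# NPC alert bounce curves provide jump, squash and tilt math.
# Custom primitive data supplies start times and other per-instance timing.
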Material /FNGameplayCues/Athena/UI/InGame/NPCState/Flee/M_NPCState_Flee_01
graph {
 properties {
  bEnableResponsiveAA=true
  blendMode=BLEND_Translucent
  numCustomizedUVs=1
  opacityMaskClipValue=0.5
  pixelDepthOffsetMode=PDOM_Legacy
  shadingModel=MSM_Unlit
  usedShadingModels=Unlit
 }
 nodes [
  n1 = TextureCoordinate()
  n2 = ScalarParameter() {bUseCustomPrimitiveData=true, parameterName=BounceStartTime, primitiveDataIndex=12}
  n3 = CurveAtlasRowParameter(CurveTime=Add(A=@n2, B=Time() {bOverride_Period=true, period=1})) {
   atlas=/FNGameplayCues/Athena/Items/Consumables/BaconSnack/NPCAlertStatesCurves
   curve=/FNGameplayCues/Athena/Items/Consumables/BaconSnack/NPCAlertBounceY
   defaultValue=1
   parameterName=CurveParams
  }
  n4 = Constant() {r=1}
  n5 = If(A=Constant() {r=1}, B=@n4, "A > B"=@n4, "A == B"=@n4, "A < B"=ScalarParameter() {defaultValue=0.25, parameterName=SuspicionAnimDampen})
  n6 = Multiply(A=ConstantBiasScale(None=@n3.A) {bias=-1, scale=1}, B=@n5)
  Add(A=@n1, B=AppendVector(A=Constant(), B=@n6)) {desc=Jump}
  n7 = MaterialFunctionCall!(
   WorldPosition
   PivotLocation
   WS Normals
   Custom Object Basis 1=Transform(None=Constant3Vector() {constant=<rgba 0 0 -1 1>}) {transformSourceType=TRANSFORMSOURCE_Local}
   Custom Object Basis 2=Transform(None=Constant3Vector() {constant=<rgba -1 0 0 1>}) {transformSourceType=TRANSFORMSOURCE_Local}
   Custom Object Basis 3=Transform(None=Constant3Vector() {constant=<rgba 0 1 0 1>}) {transformSourceType=TRANSFORMSOURCE_Local}
  ) {materialFunction=/Engine/Functions/Engine_MaterialFunctions02/WorldPositionOffset/AlignMeshToTheCamera}
  n8 = Subtract(
   A=Add(A=@n7.World Position Offset, B=WorldPosition().XYZ)
   B=MaterialFunctionCall() {materialFunction=&ObjectPivotPoint}.Object Pivot Location
  )
  n9 = Distance(
   A=MaterialFunctionCall() {materialFunction=&ObjectPivotPoint}.Object Pivot Location
   B=CameraPositionWS()
  )
  n10 = Divide(A=@n9, B=ScalarParameter() {defaultValue=10000, group=Scale, parameterName=ScaleByDistanceDivisor}) {constB=1}
  n11 = Add(
   A=Add(A=@n7.World Position Offset, B=Multiply(A=@n8, B=@n10))
   B=Multiply(A=@n9, B=VectorParameter() {group=Scale, parameterName=BiasByDistanceDivisor, defaultValue=<rgba 0 0 0.01 0>}.RGB)
  )
  LinearInterpolate(
   A=Subtract(
    A=MaterialFunctionCall() {materialFunction=&ObjectPivotPoint}.Object Pivot Location
    B=WorldPosition().XYZ
   )
   B=@n11
  ) {constAlpha=1}
  n12 = ScalarParameter() {
   bUseCustomPrimitiveData=true
   defaultValue=1
   desc=Percentage of End of Animation
   group=Animation
   parameterName=FadeOutDuration
   primitiveDataIndex=15
  }
  LinearInterpolate(Alpha=MaterialFunctionCall(TimestampA=ScalarParameter() {bUseCustomPrimitiveData=true, parameterName=FadeOutStartTime, primitiveDataIndex=14}, TransitionLength_In=@n12) {materialFunction=&MF_Timestamp_InOut_01}.SmoothStep) {constAlpha=0, constB=0}
  n13 = ScalarParameter() {bUseCustomPrimitiveData=true, parameterName=SweatStartTime, primitiveDataIndex=12}
  n14 = Time()
  n15 = Frac(None=Multiply(A=Add(A=@n13, B=@n14), B=ScalarParameter() {defaultValue=2, group=Jitter, parameterName=Speed, sliderMax=1, sortPriority=2}))
  n16 = Multiply(A=ScalarParameter() {defaultValue=25, desc=Percentage of End of Animation, group=Animation, parameterName=FadeOutPercentage}) {constB=0.01}
  n17 = OneMinus(None=@n16)
  n18 = Multiply(A=Clamp(None=Subtract(A=@n15, B=@n17) {constB=0.75}), B=Divide(B=@n16)) {constB=4}
  MaterialFunctionCall(Number=@n18, MaximumNumberOfDigits=Constant() {r=3}) {materialFunction=&DebugScalarValues}
  n19 = VectorParameter() {group=Scale, parameterName=UVScale, defaultValue=<rgba 1 1 0 0>}
  n20 = AppendVector(A=@n19.R, B=@n19.R)
  MaterialFunctionCall(
   UVs=MaterialFunctionCall!(
    Texture Scale=@n20
    UVs=@n1
    Offset XY Texturesize Z=VectorParameter() {group=Jitter, parameterName=UVOffset, defaultValue=<rgba 0.5 0.5 1 1>}.RGB
   ) {materialFunction=/Engine/Functions/Engine_MaterialFunctions01/Texturing/ScaleUVsAroundPoint}.UVs
   Rotation Center=AppendVector(A=Constant() {r=0.5}, B=OneMinus(None=@n6))
   "Rotation Angle (0-1)"=Multiply(A=ConstantBiasScale(None=@n3.B) {bias=-1, scale=1}, B=@n5)
  ) {materialFunction=&CustomRotator}
  n21 = TextureSampleParameter2D() {
   group=Color
   parameterName=Mask Texture
   samplerSource=SSM_Clamp_WorldGroupSettings
   texture=&/T_NPCState_Flee_01_D
  }
  MaterialFunctionCall(f4=@n21.RGBA) {materialFunction=&AddComponents}
  MaterialFunctionCall(
   TimestampA=ScalarParameter() {bUseCustomPrimitiveData=true, parameterName=TimestampA, primitiveDataIndex=4}
   TransitionLength_In=ScalarParameter() {bUseCustomPrimitiveData=true, parameterName=TransitionLength_In, primitiveDataIndex=5}
  ) {materialFunction=&MF_Timestamp_InOut_01}
  MaterialFunctionCall(
   TimestampA=ScalarParameter() {bUseCustomPrimitiveData=true, parameterName=SweatStartTime, primitiveDataIndex=12}
   TransitionLength_In=ScalarParameter() {defaultValue=1, parameterName=SweatCurveDuration}
  ) {materialFunction=&MF_Timestamp_InOut_01}
  MaterialFunctionCall(
   TimestampA=Add(
    A=ScalarParameter() {bUseCustomPrimitiveData=true, parameterName=SweatStartTime, primitiveDataIndex=12}
    B=ScalarParameter() {defaultValue=1, parameterName=SweatCurveDuration}
   )
   TransitionLength_In=ScalarParameter() {defaultValue=25, desc=Percentage of End of Animation, group=Animation, parameterName=FadeOutPercentage}
  ) {materialFunction=&MF_Timestamp_InOut_01}
  MaterialFunctionCall(TimestampA=@n2, TransitionLength_In=ScalarParameter() {defaultValue=1.5, parameterName=BounceDuration}) {materialFunction=&MF_Timestamp_InOut_01}
  n22 = Constant()
  n23 = Add(A=ScalarParameter() {bUseCustomPrimitiveData=true, parameterName=SweatStartTime, primitiveDataIndex=12}, B=Time())
  n24 = ScalarParameter() {defaultValue=2, group=Jitter, parameterName=Speed, sliderMax=1, sortPriority=2}
  n25 = CurveAtlasRowParameter(CurveTime=Frac(None=Multiply(A=@n23, B=@n24))) {
   atlas=&/CurveAtlas_NPCState_Flee
   curve=&/Curve_NPCState_Flee_SweatScaleAnim
   parameterName=CurveParams
  }
  n26 = VectorParameter() {group=Scale, parameterName=UV Center Start 01, sortPriority=1, defaultValue=<rgba 0.5 0.5 1 1>}
  n27 = Constant3Vector() {constant=<rgba 0.5 0.5 1 1>}
  n28 = TextureSampleParameter2D(UVs=MaterialFunctionCall(
    Texture Scale=LinearInterpolate(Alpha=@n25.R) {constA=0.1, desc=Max Min Range}
    Offset XY Texturesize Z=LinearInterpolate(A=@n26.RGB, B=@n27, Alpha=@n25.R)
   ) {materialFunction=&ScaleUVsAroundPoint}.UVs) {
   group=Color
   parameterName=Mask Texture
   samplerSource=SSM_Clamp_WorldGroupSettings
   texture=&T_NPCState_Flee_01_D
  }
  n29 = ScalarParameter() {defaultValue=0.2, parameterName=ExclaimEdge}
  n30 = ScalarParameter() {parameterName=EdgeSoftness}
  n31 = If(
   A=@n25.R
   B=@n22
   "A > B"=MaterialFunctionCall(Alpha=@n28.R, Min=@n29, Max=Add(A=@n29, B=@n30)) {materialFunction=&SmoothStep}.Result
   "A == B"=@n22
   "A < B"=@n22
  )
  n32 = Constant()
  n33 = VectorParameter() {group=Scale, parameterName=UV Center Start 02, sortPriority=1, defaultValue=<rgba 0.5 0.5 1 1>}
  n34 = TextureSampleParameter2D(UVs=MaterialFunctionCall(
    Texture Scale=LinearInterpolate(Alpha=@n25.G) {constA=0.2, desc=Max Min Range}
    Offset XY Texturesize Z=LinearInterpolate(A=@n33.RGB, B=@n27, Alpha=@n25.G)
   ) {materialFunction=&ScaleUVsAroundPoint}.UVs) {
   group=Color
   parameterName=Mask Texture
   samplerSource=SSM_Clamp_WorldGroupSettings
   texture=&T_NPCState_Flee_01_D
  }
  n35 = ScalarParameter() {defaultValue=0.2, parameterName=ExclaimEdge}
  n36 = ScalarParameter() {parameterName=EdgeSoftness}
  n37 = If(
   A=@n25.G
   B=@n32
   "A > B"=MaterialFunctionCall(Alpha=@n34.G, Min=@n35, Max=Add(A=@n35, B=@n36)) {materialFunction=&SmoothStep}.Result
   "A == B"=@n32
   "A < B"=@n32
  )
  n38 = Constant()
  n39 = VectorParameter() {group=Scale, parameterName=UV Center Start 03, sortPriority=1, defaultValue=<rgba 0.5 0.5 1 1>}
  n40 = TextureSampleParameter2D(UVs=MaterialFunctionCall(
    Texture Scale=LinearInterpolate(Alpha=@n25.B) {constA=0.1, desc=Max Min Range}
    Offset XY Texturesize Z=LinearInterpolate(A=@n39.RGB, B=@n27, Alpha=@n25.B)
   ) {materialFunction=&ScaleUVsAroundPoint}.UVs) {
   group=Color
   parameterName=Mask Texture
   samplerSource=SSM_Clamp_WorldGroupSettings
   texture=&T_NPCState_Flee_01_D
  }
  n41 = ScalarParameter() {defaultValue=0.2, parameterName=ExclaimEdge}
  n42 = MaterialFunctionCall(Alpha=@n40.B, Min=@n41, Max=Add(A=@n41, B=ScalarParameter() {parameterName=EdgeSoftness})) {materialFunction=&SmoothStep}
  n43 = Constant()
  n44 = VectorParameter() {group=Scale, parameterName=UV Center Start 04, sortPriority=1, defaultValue=<rgba 0.5 0.5 1 1>}
  n45 = TextureSampleParameter2D(UVs=MaterialFunctionCall(
    Texture Scale=LinearInterpolate(Alpha=@n25.A) {constA=0.2, desc=Max Min Range}
    Offset XY Texturesize Z=LinearInterpolate(A=@n44.RGB, B=@n27, Alpha=@n25.A)
   ) {materialFunction=&ScaleUVsAroundPoint}.UVs) {
   group=Color
   parameterName=Mask Texture
   samplerSource=SSM_Clamp_WorldGroupSettings
   texture=&T_NPCState_Flee_01_D
  }
  n46 = ScalarParameter() {defaultValue=0.2, parameterName=ExclaimEdge}
  n47 = Add(A=@n46, B=ScalarParameter() {parameterName=EdgeSoftness})
  n48 = Add(
   A=Add(A=Add(A=@n31, B=@n37), B=If(A=@n25.B, B=@n38, "A > B"=@n42.Result, "A == B"=@n38, "A < B"=@n38))
   B=If(
    A=@n25.A
    B=@n43
    "A > B"=MaterialFunctionCall(Alpha=@n45.A, Min=@n46, Max=@n47) {materialFunction=&SmoothStep}.Result
    "A == B"=@n43
    "A < B"=@n43
   )
  )
  n49 = LinearInterpolate(A=@n48, Alpha=@n18) {constAlpha=0, constB=0}
  Multiply(A=VectorParameter() {group=Color, parameterName=Color, defaultValue=<rgba 1 1 1 1>}.RGB, B=@n49)
  Multiply(A=Time() {bOverride_Period=true, period=5}, B=ScalarParameter() {defaultValue=1, parameterName=timespeed}) {constB=2, desc=test anim}
  Multiply(A=@n20, B=AppendVector(A=@n3.R, B=@n3.R)) {desc=Squash and Stretch}
  Reroute(None=Add(B=@n11))
  n50 = Constant3Vector() {constant=<rgba 0.5 0.5 0.5 0>}
  n51 = VectorParameter() {group=Scale, parameterName=UVScale, defaultValue=<rgba 1 1 0 0>}
  n52 = AppendVector(A=@n51.R, B=@n51.R)
  n53 = ScalarParameter() {bUseCustomPrimitiveData=true, parameterName=JitterStartTime, primitiveDataIndex=12}
  n54 = Multiply(A=Time(), B=ScalarParameter() {defaultValue=2, group=Jitter, parameterName=Speed, sliderMax=1, sortPriority=2})
  n55 = CurveAtlasRowParameter(CurveTime=Frac(None=Add(A=@n53, B=@n54))) {
   atlas=&CurveAtlas_NPCState_Flee
   curve=&/Curve_NPCState_Flee_JitterScaleAnim_01
   defaultValue=2
   group=Jitter
   parameterName=JitterScaleCurve
  }
  n56 = TextureSampleParameter2D(UVs=MaterialFunctionCall(
    Texture Scale=Multiply(A=@n52, B=@n55.R)
    Offset XY Texturesize Z=VectorParameter() {group=Jitter, parameterName=UV RotCenter 01, defaultValue=<rgba 0.5 0.5 1 1>}.RGB
   ) {materialFunction=&ScaleUVsAroundPoint}.UVs) {
   group=Color
   parameterName=Mask Texture
   samplerSource=SSM_Clamp_WorldGroupSettings
   texture=&T_NPCState_Flee_01_D
  }
  n57 = TextureSampleParameter2D(UVs=MaterialFunctionCall(
    Texture Scale=Multiply(A=@n52, B=@n55.G)
    Offset XY Texturesize Z=VectorParameter() {group=Jitter, parameterName=UV RotCenter 02, defaultValue=<rgba 0.5 0.5 1 1>}.RGB
   ) {materialFunction=&ScaleUVsAroundPoint}.UVs) {
   group=Color
   parameterName=Mask Texture
   samplerSource=SSM_Clamp_WorldGroupSettings
   texture=&T_NPCState_Flee_01_D
  }
  n58 = Add(A=@n56.R, B=@n57.G)
  n59 = TextureSampleParameter2D(UVs=MaterialFunctionCall(
    Texture Scale=Multiply(A=@n52, B=@n55.B)
    Offset XY Texturesize Z=VectorParameter() {group=Jitter, parameterName=UV RotCenter 03, defaultValue=<rgba 0.5 0.5 1 1>}.RGB
   ) {materialFunction=&ScaleUVsAroundPoint}.UVs) {
   group=Color
   parameterName=Mask Texture
   samplerSource=SSM_Clamp_WorldGroupSettings
   texture=&T_NPCState_Flee_01_D
  }
  n60 = TextureSampleParameter2D(UVs=MaterialFunctionCall(
    Texture Scale=Multiply(A=@n52, B=@n55.A)
    Offset XY Texturesize Z=VectorParameter() {group=Jitter, parameterName=UV RotCenter 04, defaultValue=<rgba 0.5 0.5 1 1>}.RGB
   ) {materialFunction=&ScaleUVsAroundPoint}.UVs) {
   group=Color
   parameterName=Mask Texture
   samplerSource=SSM_Clamp_WorldGroupSettings
   texture=&T_NPCState_Flee_01_D
  }
  n61 = VectorParameter() {group=Scale, parameterName=UVScale, defaultValue=<rgba 1 1 0 0>}
  n62 = AppendVector(A=@n61.R, B=@n61.R)
  n63 = MaterialFunctionCall(Texture Scale=@n62) {materialFunction=&ScaleUVsAroundPoint}
  n64 = ScalarParameter() {bUseCustomPrimitiveData=true, parameterName=JitterStartTime, primitiveDataIndex=12}
  n65 = Multiply(A=Time(), B=ScalarParameter() {defaultValue=2, group=Jitter, parameterName=Speed, sliderMax=1, sortPriority=2})
  n66 = CurveAtlasRowParameter(CurveTime=Frac(None=Add(A=@n64, B=@n65))) {
   atlas=&CurveAtlas_NPCState_Flee
   curve=&/Curve_NPCState_Flee_JitterRotAnim_01
   defaultValue=1
   group=Jitter
   parameterName=JitterCurve
  }
  n67 = ConstantBiasScale(None=@n66.R) {bias=-1, scale=1}
  n68 = ScalarParameter() {defaultValue=0.25, group=Jitter, parameterName=AnimDampen, sortPriority=5}
  n69 = TextureSampleParameter2D(UVs=MaterialFunctionCall(
    UVs=@n63.UVs
    Rotation Center=VectorParameter() {group=Jitter, parameterName=UV RotCenter 01, defaultValue=<rgba 0.5 0.5 1 1>}.RGB
    "Rotation Angle (0-1)"=Multiply(A=@n67, B=@n68)
   ) {materialFunction=&CustomRotator}.Rotated Values) {
   group=Color
   parameterName=Mask Texture
   samplerSource=SSM_Clamp_WorldGroupSettings
   texture=&T_NPCState_Flee_01_D
  }
  n70 = ConstantBiasScale(None=@n66.G) {bias=-1, scale=1}
  n71 = TextureSampleParameter2D(UVs=MaterialFunctionCall(
    UVs=@n63.UVs
    Rotation Center=VectorParameter() {group=Jitter, parameterName=UV RotCenter 02, defaultValue=<rgba 0.5 0.5 1 1>}.RGB
    "Rotation Angle (0-1)"=Multiply(A=@n70, B=@n68)
   ) {materialFunction=&CustomRotator}.Rotated Values) {
   group=Color
   parameterName=Mask Texture
   samplerSource=SSM_Clamp_WorldGroupSettings
   texture=&T_NPCState_Flee_01_D
  }
  n72 = Add(A=@n69.R, B=@n71.G)
  n73 = ConstantBiasScale(None=@n66.B) {bias=-1, scale=1}
  n74 = TextureSampleParameter2D(UVs=MaterialFunctionCall(
    UVs=@n63.UVs
    Rotation Center=VectorParameter() {group=Jitter, parameterName=UV RotCenter 03, defaultValue=<rgba 0.5 0.5 1 1>}.RGB
    "Rotation Angle (0-1)"=Multiply(A=@n73, B=@n68)
   ) {materialFunction=&CustomRotator}.Rotated Values) {
   group=Color
   parameterName=Mask Texture
   samplerSource=SSM_Clamp_WorldGroupSettings
   texture=&T_NPCState_Flee_01_D
  }
  n75 = ConstantBiasScale(None=@n66.A) {bias=-1, scale=1}
  n76 = TextureSampleParameter2D(UVs=MaterialFunctionCall(
    UVs=@n63.UVs
    Rotation Center=VectorParameter() {group=Jitter, parameterName=UV RotCenter 04, defaultValue=<rgba 0.5 0.5 1 1>}.RGB
    "Rotation Angle (0-1)"=Multiply(A=@n75, B=@n68)
   ) {materialFunction=&CustomRotator}.Rotated Values) {
   group=Color
   parameterName=Mask Texture
   samplerSource=SSM_Clamp_WorldGroupSettings
   texture=&T_NPCState_Flee_01_D
  }
  n77 = StaticSwitchParameter(True=Add(A=Add(A=@n58, B=@n59.B), B=@n60.A), False=Add(A=Add(A=@n72, B=@n74.B), B=@n76.A)) {group=Jitter, parameterName=Use Scale Jitter, sortPriority=0}
  Reroute(None=StaticSwitchParameter(True=LinearInterpolate(A=@n50, Alpha=@n77) {constA=0.5}, False=@n77) {group=Color, parameterName=UV Bounds Debug})
  n78 = Constant()
  n79 = CurveAtlasRowParameter(CurveTime=MaterialFunctionCall(TimestampA=@n64, TransitionLength_In=ScalarParameter() {defaultValue=0.5, group=Jitter, parameterName=Jump Intro Duration}) {materialFunction=&MF_Timestamp_InOut_01}.SmoothStep) {
   atlas=&NPCAlertStatesCurves
   curve=&NPCAlertBounceY
   parameterName=CurveParams
  }
  n80 = Multiply(A=ConstantBiasScale(None=@n79.A) {bias=-1, scale=1})
  Reroute(None=Add(A=TextureCoordinate(), B=AppendVector(A=@n78, B=@n80)) {desc=Jump})
  Reroute(None=Multiply(A=@n62, B=AppendVector(A=@n79.R, B=@n79.R)) {desc=Squash and Stretch})
  ScalarParameter() {bUseCustomPrimitiveData=true, defaultValue=1, parameterName=IconState, primitiveDataIndex=3, sliderMax=1}
  StaticSwitchParameter(True=@n49, False=@n48) {defaultValue=true, group=Animation, parameterName=EnableFadeOut}
  Time() {bOverride_Period=true, desc=Debug. Doesn't work in Game, period=0.6}
  VectorParameter() {group=Jitter, parameterName=UVOffset, defaultValue=<rgba 0.5 0.5 1 1>}
  VectorParameter() {group=Jitter, parameterName=UVOffset, defaultValue=<rgba 0.5 0.5 1 1>}
 ]
}
